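# Conversation database for the Wasteland Warrior NPC, defining the greeting that opens a conversation with a player.
# The entry tag leads to one speech node with a default line plus contextual variants for first-ever meetings, match conversation count checks and specific player outfits.
# First-meeting and outfit lines take their priority from named curve rows.
# The greeting branches through link nodes into the NPC library quests, basic services and the Wasteland Warrior dialogue entry.
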ConversationDatabase /NPCLibrary/NPCs/WastelandWarrior/NPCLibrary_Conversation_WastelandWarrior
entry_points [{tag=Conversation.Entry.WastelandWarrior, destinations [#1]}]
nodes [
 ConversationEntryPointNode_7 {
  id=#1
  class=ConversationEntryPointNode
  outputs [#2]
  properties {entryTag {tagName=Conversation.Entry.WastelandWarrior}}
 }
 Speech_C_4 {
  id=#2
  class=Speech_C
  outputs [#3, #4, #5]
  properties {
   nodeName=""
   generalConfig {
    defaultMessage=Looking to change the world?
    contextualMessages [
     {
      message="Let's change the world, yeah?"
      contextRequirements [{requirement {class=FortControllerRequirement_IsFirstEverConversationWithNPC}}]
      priority {curve {rowName=Tandem.Speech.Priority.FirstEverInteraction}}
     }
     {
      message=Looking to change the world?
      contextRequirements [{requirement {
         class=FortControllerRequirement_CheckNPCMatchConversationCount
         properties {nPCUniqueIdTag {tagName=AISpawnerData.Type.Tandem.WastelandWarrior}}
        }}]
      priority {curve {rowName=None}}
     }
     {
      message=Please tell me you're not one of Ruckus's Rabbles.
      contextRequirements [{requirement {
         class=FortControllerRequirement_CheckNPCMatchConversationCount
         properties {nPCUniqueIdTag {tagName=AISpawnerData.Type.Tandem.WastelandWarrior}}
        }}]
      priority {curve {rowName=None}}
     }
     {
      message=You with Power Chord and Riot?
      contextRequirements [{requirement {
         class=FortControllerRequirement_CheckNPCMatchConversationCount
         properties {nPCUniqueIdTag {tagName=AISpawnerData.Type.Tandem.WastelandWarrior}}
        }}]
      priority {curve {rowName=None}}
     }
     {
      message=Backbone calls his people the Riders. You one of them?
      contextRequirements [{requirement {
         class=FortControllerRequirement_CheckNPCMatchConversationCount
         properties {nPCUniqueIdTag {tagName=AISpawnerData.Type.Tandem.WastelandWarrior}}
        }}]
      priority {curve {rowName=None}}
     }
     {
      message=Here for a little peace and quiet?
      contextRequirements [{requirement {
         class=FortControllerRequirement_CheckNPCMatchConversationCount
         properties {nPCUniqueIdTag {tagName=AISpawnerData.Type.Tandem.WastelandWarrior}}
        }}]
      priority {curve {rowName=None}}
     }
     {
      message=You seem pretty punk rock. Cool.
      contextRequirements [{requirement {
         class=FortControllerRequirement_HasCID
         properties {allowedCIds [/BRCosmetics/Athena/Items/Cosmetics/Characters/CID_097_Athena_Commando_F_RockerPunk]}
        }}]
      priority {curve {rowName=Tandem.Speech.Priority.OutfitReaction}}
     }
     {
      message="You're, uh, kind of a lot, aren't you?"
      contextRequirements [{requirement {
         class=FortControllerRequirement_HasCID
         properties {allowedCIds [/BRCosmetics/Athena/Items/Cosmetics/Characters/CID_262_Athena_Commando_M_MadCommander]}
        }}]
      priority {curve {rowName=Tandem.Speech.Priority.OutfitReaction}}
     }
     {
      message=The wasteland's lucky to have someone like you.
      contextRequirements [{requirement {
         class=FortControllerRequirement_HasCID
         properties {allowedCIds [/BRCosmetics/Athena/Items/Cosmetics/Characters/CID_196_Athena_Commando_M_Biker]}
        }}]
      priority {curve {rowName=Tandem.Speech.Priority.OutfitReaction}}
     }
     {
      message=I dig the hair. Could maybe turn the music down though.
      contextRequirements [{requirement {
         class=FortControllerRequirement_HasCID
         properties {allowedCIds [/BRCosmetics/Athena/Items/Cosmetics/Characters/CID_268_Athena_Commando_M_RockerPunk]}
        }}]
      priority {curve {rowName=Tandem.Speech.Priority.OutfitReaction}}
     }
    ]
   }
  }
 }
 ConversationLinkNode_2 {
  id=#5
  class=ConversationLinkNode
  properties {remoteEntryTag {tagName=Conversation.Entry.WastelandWarrior.Dialogue}}
 }
 ConversationLinkNode_0 {id=#4, class=ConversationLinkNode, properties {remoteEntryTag {tagName=Conversation.Entry.BasicServices}}}
 ConversationLinkNode_1 {id=#3, class=ConversationLinkNode, properties {remoteEntryTag {tagName=Conversation.Entry.NPCLibrary.Quests}}}
]
