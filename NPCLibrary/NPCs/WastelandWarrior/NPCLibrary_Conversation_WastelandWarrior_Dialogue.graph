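# Conversation database for the Wasteland Warrior NPC's "Ask about..." dialogue, giving the NPC's lore and opinions.
# One entry tag opens a prompt speech that branches into three topic choices (The Chums, Megalo Don, Anything strange?) plus a Back node.
# Each topic plays a linear chain of Speech nodes, and all three converge on one shared ReturnToStart node shown as a BACK choice.
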ConversationDatabase /NPCLibrary/NPCs/WastelandWarrior/NPCLibrary_Conversation_WastelandWarrior_Dialogue
entry_points [{tag=Conversation.Entry.WastelandWarrior.Dialogue, destinations [#1]}]
nodes [
 ConversationEntryPointNode_0 {
  id=#1
  class=ConversationEntryPointNode
  outputs [#2]
  properties {entryTag {tagName=Conversation.Entry.WastelandWarrior.Dialogue}}
 }
 Speech_C_0 {
  id=#2
  class=Speech_C
  outputs [#3, #4, #5, #6]
  properties {nodeName="", softTaskIcon=&T_UI_NPC_question, generalConfig {defaultMessage=You wanted to ask me something?}}
  sub_nodes [ConversationChoiceNode_5 {class=ConversationChoiceNode, properties {defaultChoiceDisplayText=Ask about...}}]
 }
 Speech_C_5 {
  id=#6
  class=Speech_C
  outputs [#7]
  properties {nodeName="", softTaskIcon=&T_UI_NPC_question, generalConfig {defaultMessage="Now that you mention it, yeah."}}
  sub_nodes [ConversationChoiceNode_15 {class=ConversationChoiceNode, properties {defaultChoiceDisplayText=Anything strange?}}]
 }
 Speech_C_6 {
  id=#7
  class=Speech_C
  outputs [#8]
  properties {
   nodeName=""
   softTaskIcon=&T_UI_NPC_question
   generalConfig {defaultMessage="I mean, the wasteland is strange enough as is."}
  }
 }
 Speech_C_8 {
  id=#8
  class=Speech_C
  outputs [#9]
  properties {
   nodeName=""
   softTaskIcon=&T_UI_NPC_question
   generalConfig {defaultMessage=But lately...I don't know. I feel like someone's watching me.}
  }
 }
 ReturnToStart_C_23 {
  id=#9
  class=ReturnToStart_C
  properties {softTaskIcon=&T_UI_NPC_Back}
  sub_nodes [ConversationChoiceNode_32 {class=ConversationChoiceNode, properties {defaultChoiceDisplayText=BACK}}]
 }
 Speech_C_2 {
  id=#5
  class=Speech_C
  outputs [#10]
  properties {
   nodeName=""
   softTaskIcon=&T_UI_NPC_question
   generalConfig {defaultMessage=He's...a lot. And not in the good way. If there is a good way.}
  }
  sub_nodes [ConversationChoiceNode_1 {class=ConversationChoiceNode, properties {defaultChoiceDisplayText=Megalo Don}}]
 }
 Speech_C_3 {
  id=#10
  class=Speech_C
  outputs [#11]
  properties {
   nodeName=""
   softTaskIcon=&T_UI_NPC_question
   generalConfig {defaultMessage=Seems like Megalo Don doesn't care about anything except Nitro.}
  }
 }
 Speech_C_4 {
  id=#11
  class=Speech_C
  outputs [#9]
  properties {
   nodeName=""
   softTaskIcon=&T_UI_NPC_question
   generalConfig {defaultMessage=In my experience? That can only end badly.}
  }
 }
 Back_C_0 {
  id=#4
  class=Back_C
  sub_nodes [ConversationChoiceNode_7 {class=ConversationChoiceNode, properties {defaultChoiceDisplayText=BACK}}]
 }
 Speech_C_7 {
  id=#3
  class=Speech_C
  outputs [#12]
  properties {
   nodeName=""
   softTaskIcon=&T_UI_NPC_question
   generalConfig {defaultMessage=Not a full day on this island and I got swarmed by people.}
  }
  sub_nodes [ConversationChoiceNode_0 {class=ConversationChoiceNode, properties {defaultChoiceDisplayText=The Chums}}]
 }
 Speech_C_15 {
  id=#12
  class=Speech_C
  outputs [#13]
  properties {
   nodeName=""
   softTaskIcon=&T_UI_NPC_question
   generalConfig {defaultMessage="Power Chord, Backbone, Riot, even Ruckus."}
  }
 }
 Speech_C_14 {
  id=#13
  class=Speech_C
  outputs [#14]
  properties {
   nodeName=""
   softTaskIcon=&T_UI_NPC_question
   generalConfig {defaultMessage=They all want me to lead their group or something?}
  }
 }
 Speech_C_1 {
  id=#14
  class=Speech_C
  outputs [#9]
  properties {
   nodeName=""
   softTaskIcon=&T_UI_NPC_question
   generalConfig {defaultMessage="I don't know, maybe. People need to stick together in the wasteland."}
  }
 }
]
